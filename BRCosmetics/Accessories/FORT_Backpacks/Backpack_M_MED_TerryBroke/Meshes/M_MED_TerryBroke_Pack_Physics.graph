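# Physics asset for the M_MED_TerryBroke backpack cosmetic, defining secondary motion for its pendant, ropes and goods.
# Capsule bodies cover the pack, two master anchor bones and the dyn_ bones.
# Constraints form a three-link pendant chain from MSTR_pendant and two-link left and right rope chains from MSTR_rope, while the goods hang off the pack.
# Per-joint swing limits shape the motion, with twist locked or nearly locked throughout.
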PhysicsAsset /BRCosmetics/Accessories/FORT_Backpacks/Backpack_M_MED_TerryBroke/Meshes/M_MED_TerryBroke_Pack_Physics
bodies {
 MSTR_pendant [{shapeType=Capsule, length=0.1, radius=0.1, shapeName=MSTR_pendant_capsule}]
 MSTR_rope [{shapeType=Capsule, length=0.505, radius=0.505, shapeName=MSTR_rope_capsule, rotation=<rot 0 -90 0>}]
 dyn_goods [{shapeType=Capsule, length=0.5491, radius=7.2503, shapeName=dyn_goods_capsule, center=<-3 0 0>, rotation=<rot 90 0 0>}]
 dyn_pendant_01 [{shapeType=Capsule, length=2.3045, radius=1.8579, shapeName=dyn_pendant_01_capsule, center=<2 0 0>, rotation=<rot -90 0 0>}]
 dyn_pendant_02 [{shapeType=Capsule, length=1.8958, radius=1.8435, shapeName=dyn_pendant_02_capsule, center=<1 0 0>, rotation=<rot -90 0 0>}]
 dyn_pendant_03 [{shapeType=Capsule, length=3.8596, radius=1.0662, shapeName=dyn_pendant_03_capsule, center=<2.7 0 0.15>, rotation=<rot -90 0 0>}]
 dyn_rope_01_l [{shapeType=Capsule, length=1.4578, radius=0.5528, shapeName=dyn_rope_01_l_capsule, center=<1.2 0 0.15>, rotation=<rot -90 0 0>}]
 dyn_rope_01_r [{shapeType=Capsule, length=2.6914, radius=0.7032, shapeName=dyn_rope_01_r_capsule, center=<1.8 0 -0.11>, rotation=<rot -90 0 0>}]
 dyn_rope_02_l [{shapeType=Capsule, length=2.0272, radius=0.5557, shapeName=dyn_rope_02_l_capsule, center=<1.576 -0.55 0.7>, rotation=<rot -69.1051 0 0.0001>}]
 dyn_rope_02_r [{
   shapeType=Capsule
   length=2.1286
   radius=0.5509
   shapeName=dyn_rope_02_r_capsule
   center=<1.5801 0 0.3>
   rotation=<rot -76.6819 -0.0002 0.0002>
  }]
 pack [{shapeType=Capsule, length=0.918, radius=8.8863, center=<0 -10.6718 -7.2101>}]
}
constraints [
 {bone1Name=dyn_pendant_01, bone2Name=MSTR_pendant, swing1LimitDegrees=15, swing2LimitDegrees=35, twistLimitDegrees=5}
 {bone1Name=dyn_pendant_02, bone2Name=dyn_pendant_01, swing1LimitDegrees=20, swing2LimitDegrees=20, twistLimitDegrees=5}
 {bone1Name=dyn_pendant_03, bone2Name=dyn_pendant_02, swing1LimitDegrees=4, swing2LimitDegrees=70, twistLimitDegrees=5}
 {bone1Name=dyn_goods, bone2Name=pack, swing1LimitDegrees=7, swing2LimitDegrees=7, twistLimitDegrees=0}
 {bone1Name=dyn_rope_01_l, bone2Name=MSTR_rope, swing1LimitDegrees=35, swing2LimitDegrees=45, twistLimitDegrees=0}
 {bone1Name=dyn_rope_02_l, bone2Name=dyn_rope_01_l, swing1LimitDegrees=35, swing2LimitDegrees=45, twistLimitDegrees=0}
 {bone1Name=dyn_rope_01_r, bone2Name=MSTR_rope, swing1LimitDegrees=25, swing2LimitDegrees=45, twistLimitDegrees=0}
 {bone1Name=dyn_rope_02_r, bone2Name=dyn_rope_01_r, swing1LimitDegrees=35, swing2LimitDegrees=40, twistLimitDegrees=0}
]
