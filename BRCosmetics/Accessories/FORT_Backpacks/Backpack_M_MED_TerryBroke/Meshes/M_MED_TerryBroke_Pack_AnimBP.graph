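# Animation Blueprint for the TerryBroke backpack cosmetic that adds physics-driven secondary motion to the pack.
# A RigidBody node simulates the input pose using the pack's physics asset, based at the pack bone, with an overridden world gravity.
# Each update eases that gravity's Z toward an upward value while skydive floating and a downward value otherwise, so loose parts lift or hang with the state.
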AnimBlueprint /BRCosmetics/Accessories/FORT_Backpacks/Backpack_M_MED_TerryBroke/Meshes/M_MED_TerryBroke_Pack_AnimBP
target_skeleton=&/M_MED_TerryBroke_Pack_Skeleton
graphs [
 AnimGraph {nodes [
   n1 = RigidBody!(
    ComponentPose=Local To Component(LocalPose=Input Pose())
    Alpha
    OverrideWorldGravity:Vector=Get Override World Gravity() -> (Override World Gravity:Vector)
    ExternalForce
   ) -> (Pose) {
    bClampLinearTranslationLimitToRefPose=true
    bOverrideWorldGravity=true
    overridePhysicsAsset=&/M_MED_TerryBroke_Pack_Physics
    baseBoneRef {boneName=pack}
    overrideWorldGravity {z=540}
    simSpaceSettings {dampingAlpha=0, maxAngularAcceleration=32, maxAngularVelocity=8, maxLinearAcceleration=1200, maxLinearVelocity=700, worldAlpha=1}
   }
   Output Pose(Result=Component To Local(ComponentPose=@n1)) {layerGroup=None}
  ]}
 EventGraph {nodes [
   n1 = Event BlueprintUpdateAnimation()
   > Set Delta Time X(Delta Time X=@n1.DeltaTimeX)
   n2 = Get bIsSkydiveFloating()
   > Set!(
    execute
    Override World Gravity_X:double=0
    Override World Gravity_Y:double=0
    Override World Gravity_Z:double=FInterpTo(
     Current=Get!() -> (Override World Gravity_X:double, Override World Gravity_Y:double, Override World Gravity_Z:double).Override World Gravity_Z
     Target=SelectFloat(A=420, B=-980, bPickA=@n2)
     DeltaTime=Get Delta Time X()
     InterpSpeed=2
    )
   ) -> (then, Output_Get:Vector)
   TryGetPawnOwner()
  ]}
]
variables [
 CurentGravityValue {type=Vector}
 Delta Time X {type=number}
 Override World Gravity {type=Vector, default=<0 0 540>}
]
